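# Looping thrust sound cue for the OGVault glider cosmetic, using the shared glider thrust attenuation, glider loop properties sheet and glider sound class.
# A mixer layers a wind loop with a tonal loop, the tonal layer weighted lower.
# Each layer has a continuous modulator driven by the AnyDirection parameter, scaling volume and slightly shifting pitch, with a wider pitch range on the wind.
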SoundCue /BRCosmetics/Sounds/Gliders/OGVault/GliderThrustLoop_OGVault_Cue
root=Mixer_0
nodes [
 {id=Mixer_0, children [ModulatorContinuous_0, ModulatorContinuous_1], properties {inputVolume [1,0.5]}}
 {
  id=ModulatorContinuous_0
  children [WavePlayer_0]
  properties {
   pitchModulationParams {default=0, maxOutput=1.1, minOutput=1, parameterName=AnyDirection}
   volumeModulationParams {default=0, minOutput=0.5, parameterName=AnyDirection}
  }
 }
 {
  id=ModulatorContinuous_1
  children [WavePlayer_1]
  properties {
   pitchModulationParams {default=0, minOutput=0.99, parameterName=AnyDirection}
   volumeModulationParams {default=0, minOutput=0.6, parameterName=AnyDirection}
  }
 }
 {id=WavePlayer_0, properties {bLooping=true, soundWaveAssetPtr=&/Glider_OGVault_Thrust_LP_Wind}}
 {id=WavePlayer_1, properties {bLooping=true, soundWaveAssetPtr=&/Glider_OGVault_Thrust_LP_Tonal}}
]
properties {
 attenuationSettings=&Fort_Player_Glider_Thrust_Athena_Att
 audioPropertiesSheet=&APS_Cosmetic_Glider_Loop
 maxDistance=12500
 soundClassObject=&Fort_Gliders_Land
 volumeMultiplier=1
 concurrencySet [null]
}
